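# Physics asset for the male medium HornedJudgment Valhalla player body cosmetic, covering a dynamic right-side horn and the body parts around it.
# Capsule bodies on the pelvis, right thigh, horn master bone and horn bone form a single constraint chain.
# The horn master rotates freely on all angular axes relative to the thigh, while the horn bone is held to it by very tight swing and twist limits.
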PhysicsAsset /BRCosmetics/Characters/Player/Male/Medium/Bodies/M_MED_HornedJudgment_Valhalla/Meshes/M_MED_HornedJudgment_Valhalla_Physics
bodies {
 dyn_horn_r_01 [{shapeType=Capsule, length=6.5, radius=5.8, shapeName=dyn_horn_r_01_capsule, center=<9.6876 -3 0>}]
 dyn_horn_r_MSTR [{shapeType=Capsule, length=0.5, radius=0.5, shapeName=dyn_horn_r_MSTR_capsule}]
 pelvis [{shapeType=Capsule, length=10, radius=11.5}]
 thigh_r [{shapeType=Capsule, length=23, radius=10, center=<12.5225 0 0>, rotation=<rot 90 0 0>}]
}
constraints [
 {bone1Name=thigh_r, bone2Name=pelvis, swing1LimitDegrees=45, swing2LimitDegrees=45, twistLimitDegrees=45}
 {bone1Name=dyn_horn_r_01, bone2Name=dyn_horn_r_MSTR, swing1LimitDegrees=2, swing2LimitDegrees=4, twistLimitDegrees=0}
 {
  bone1Name=dyn_horn_r_MSTR
  bone2Name=thigh_r
  swing1LimitDegrees=45
  swing1Motion=Free
  swing2LimitDegrees=45
  swing2Motion=Free
  twistLimitDegrees=45
  twistMotion=Free
 }
]
