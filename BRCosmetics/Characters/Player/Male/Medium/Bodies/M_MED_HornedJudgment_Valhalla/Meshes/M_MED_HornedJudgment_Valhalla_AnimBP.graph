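# Animation Blueprint for the HornedJudgment Valhalla male medium body cosmetic, following the attached parent mesh's pose.
# A skydive-aware Control Rig feeds a right-horn bone constraint, head and neck adjustments, then pelvis-space RigidBody physics.
# The event graph derives head rotation from negated, per-axis scaled head-to-spine_05 delta, with scale zeroed during emotes.
# Both neck bones get a fixed yaw while skydive floating.
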AnimBlueprint /BRCosmetics/Characters/Player/Male/Medium/Bodies/M_MED_HornedJudgment_Valhalla/Meshes/M_MED_HornedJudgment_Valhalla_AnimBP
target_skeleton=&/M_MED_HornedJudgment_Valhalla_Skeleton
graphs [
 AnimGraph {nodes [
   n1 = "Constraint - dyn_horn_r_MSTR (PPPP)"!(
    ComponentPose=Local To Component(LocalPose=Use cached pose 'CRCache'())
    Alpha=1
    ConstraintWeights_0=0.4602
    ConstraintWeights_1:float=0.2685
    ConstraintWeights_2:float=0.2642
    ConstraintWeights_3:float=0.0072
   ) -> (Pose) {
    alphaBoolBlend {blendOption=Linear}
    boneToModify {boneName=dyn_horn_r_MSTR}
    constraintSetup [
     {targetBone {boneName=pelvis}}
     {targetBone {boneName=deform_glute_r}}
     {targetBone {boneName=thigh_twist_01_r}}
     {targetBone {boneName=deform_groin_r}}
    ]
    constraintWeights [1,1,1,1]
   }
   n2 = "Transform (Modify) Bone - Bone: head"(ComponentPose=@n1, Alpha=1, Translation="0,0,0", Rotation=Get head rot() -> (head rot:Rotator), Scale="1,1,1") {rotationMode=BMM_Replace, rotationSpace=BCS_BoneSpace, alphaBoolBlend {blendOption=Linear}, boneToModify {boneName=head}}
   n3 = "Transform (Modify) Bone - Bone: neck_01"!(ComponentPose=@n2, bAlphaBoolEnabled:bool=Get bIsSkydiveFloating(), Translation="0,0,0", Rotation="0,10,0", Scale="1,1,1") -> (Pose) {
    alphaInputType=Bool
    rotationMode=BMM_Replace
    rotationSpace=BCS_BoneSpace
    translationMode=BMM_Additive
    translationSpace=BCS_ParentBoneSpace
    alphaBoolBlend {blendOption=Linear}
    boneToModify {boneName=neck_01}
    rotation {yaw=10}
   }
   n4 = "Transform (Modify) Bone - Bone: neck_02"!(ComponentPose=@n3, bAlphaBoolEnabled:bool=Get bIsSkydiveFloating(), Translation="0,0,0", Rotation="0,10,0", Scale="1,1,1") -> (Pose) {
    alphaInputType=Bool
    rotationMode=BMM_Replace
    rotationSpace=BCS_BoneSpace
    translationMode=BMM_Additive
    translationSpace=BCS_ParentBoneSpace
    alphaBoolBlend {blendOption=Linear}
    boneToModify {boneName=neck_02}
    rotation {yaw=10}
   }
   n5 = RigidBody(ComponentPose=@n4) {
    bClampLinearTranslationLimitToRefPose=true
    bOverrideWorldGravity=true
    overridePhysicsAsset=&/M_MED_HornedJudgment_Valhalla_Physics
    simulationSpace=BaseBoneSpace
    alphaBoolBlend {blendOption=Linear}
    baseBoneRef {boneName=pelvis}
    componentAppliedLinearAccClamp=<6000 6000 6000>
    componentLinearAccScale=<0.5 0.5 0.5>
    componentLinearVelScale=<0.5 0.5 0.5>
    overrideWorldGravity {z=-980}
   }
   Output Pose(Result=Component To Local(ComponentPose=@n5)) {layerGroup=None}
   n6 = Control Rig!(Source=Copy Pose From Mesh() {bUseAttachedParent=true}, Alpha, isSkydiving:bool=Get bIsSkyDiving()) -> (Pose) {
    alphaBoolBlend {blendOption=Linear}
    controlRigAssetReference {blueprintRigClass=&/M_MED_Horned_Judgement_Valhalla_ControlRig1.M_MED_Horned_Judgement_Valhalla_ControlRig1_C}
   }
   Save cached pose 'CRCache'(Pose=@n6) {cacheName=CRCache}
  ]}
 EventGraph {nodes [
   Event BlueprintUpdateAnimation()
   > n1 = Sequence()
   > n2 = Sequence!(execute) -> (then_0, then_1, then_2:Exec, then_3:Exec, then_4:Exec, then_5:Exec, then_6:Exec)
   n3 = Branch(execute=@n2.then_6, Condition=Get bIsPlayingEmote())
   > n4 = Set headRotScale(headRotScale:Rotator="0, 0, 0") -> (Output_Get:Rotator)
   n5 = Set headRotScale(execute=@n3.else, headRotScale="0.3,0.6,0.6")
   n6 = BreakRotator(InRot=@n4)
   Set X percentage(execute=@n2.then_0, X percentage:float=@n6) -> (Output_Get:float)
   Set Y percentage(execute=@n2.then_1, Y percentage:float=@n6.Pitch) -> (Output_Get:float)
   Set Z percentage(execute=@n2.then_2, Z percentage:float=@n6.Yaw) -> (Output_Get:float)
   n7 = BreakRotator(InRot=@n5)
   Set X percentage(execute=@n2.then_3, X percentage=@n7)
   Set Y percentage(execute=@n2.then_4, Y percentage=@n7.Pitch)
   Set Z percentage(execute=@n2.then_5, Z percentage=@n7.Yaw)
   n8 = GetOwnerSkeletalMeshComponent()
   n9 = BreakRotator(InRot=BreakTransform(InTransform=GetDeltaTransformFromRefPose(self=@n8, BoneName=head, BaseName=spine_05)).Rotation)
   n10 = Get Y percentage() -> (Y percentage:float)
   n11 = Get X percentage() -> (X percentage:float)
   n12 = Get Z percentage() -> (Z percentage:float)
   Set head rot(
    execute=@n1.then_1
    head rot:Rotator=MakeRotator(
     Roll=float * float!(A=@n9, B=-1, C:double=@n11) -> (ReturnValue)
     Pitch=float * float(A=@n9.Pitch, B=-1, C=@n10)
     Yaw=float * float(A=@n9.Yaw, B=-1, C=@n12)
    )
   ) -> (Output_Get:Rotator)
  ]}
]
variables [
 head rot {type=Rotator}
 headRotScale {type=Rotator}
 X percentage {type=number}
 Y percentage {type=number}
 Z percentage {type=number}
]
